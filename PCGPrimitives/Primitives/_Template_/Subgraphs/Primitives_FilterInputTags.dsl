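; PCG subgraph for the Primitives template that filters input data by a tag string supplied as a parameter.
; A compare on FilteredInputTags drives a Branch that either passes data straight to Out or routes it through Filter Data By Tag, likely bypassing when no tags are set.
; InverFilter sets the filter operation, and only the filter's InsideFilter result reaches Out.
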(graph Primitives_FilterInputTags
  (params (FilteredInputTags :type String) (InverFilter :type Bool))
  (bind AttributeCompareOp_0 (Attribute%20Compare%20Op :InA FilteredInputTags))
  (bind Branch_1 (Branch))
  (bind FilterDataByTag_1 (Filter%20Data%20By%20Tag :Operation-param InverFilter :SelectedTags-param FilteredInputTags))
  (connect AttributeCompareOp_0 Out Branch_1 bOutputToB)
  (connect DefaultInputNode In Branch_1 In)
  (connect Branch_1 Output%20A FilterDataByTag_1 In)
  (output Out Branch_1.Output%20B)
  (output Out FilterDataByTag_1.InsideFilter))
