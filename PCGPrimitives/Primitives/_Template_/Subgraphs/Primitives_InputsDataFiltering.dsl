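; PCG template subgraph that sorts mixed input references into typed outputs for points, open splines, closed splines, polygons, surfaces, textures and volumes.
; Type filters split the data, with splines further divided by their IsClosed attribute.
; Data lacking its matching "Type:" tag gets known type tags replaced with the correct one.
; Branches gate each output on its InputsAllow flag and data count; InputsAllowRawdatas is declared but unused.
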(graph Primitives_InputsDataFiltering
  (params (InputsAllowPoints :type Bool :default true) (InputsAllowSplines :type Bool :default true) (InputsAllowClosedSplines :type Bool :default true) (InputsAllowPolygon2Ds :type Bool :default true) (InputsAllowSurfaces :type Bool :default true) (InputsAllowTexture2Ds :type Bool :default true) (InputsAllowVolumes :type Bool :default true) (InputsAllowRawdatas :type Bool :default true))
  (inputs (InputReferences :type Any))
  (outputs (Points :type Point) (Splines :type Spline) (ClosedSplines :type Spline) (Polygon2Ds :type Polygon2D) (Surfaces :type Surface) (Texture2Ds :type BaseTexture) (Volumes :type Volume))
  (bind FilterDataByType_8 (Filter%20Data%20By%20Type :TargetType (struct :Ids [(struct :Struct /Script/PCG.PCGDataTypeInfoSpline)])))
  (bind FilterDataByAttribute_1 (Filter%20Data%20By%20Attribute :FilterMode FilterByValue :FilterOperator NotEqual :TargetAttribute @Data.$IsClosed :Threshold (struct :AttributeTypes (struct :Type Boolean :BoolValue true))))
  (bind FilterDataByTag_3 (Filter%20Data%20By%20Tag :SelectedTags "Type:ClosedSpline"))
  (bind FilterDataByTag_4 (Filter%20Data%20By%20Tag :SelectedTags "Type:Spline"))
  (bind FilterDataByType_7 (Filter%20Data%20By%20Type :TargetType (struct :Ids [(struct :Struct /Script/PCG.PCGDataTypeInfoPoint)])))
  (bind FilterDataByTag_9 (Filter%20Data%20By%20Tag :SelectedTags "Type:Point"))
  (bind DeleteTags_2 (Delete%20Tags :Operator Substring :SelectedTags "Type:Point, Type:Spline, Type:ClosedSpline, Type:Polygon2D, Type:Surface, Type:Texture2D, Type:Volume"))
  (bind AddTags_1 (Add%20Tags :TagsToAdd "Type:Point"))
  (bind Gather_0 (Gather))
  (bind DataNum_4 (Data%20Count))
  (bind AttributeBooleanOp_12 (Attribute%20Boolean%20Op :InA InputsAllowPoints))
  (bind Branch_8 (Branch))
  (bind DeleteTags_4 (Delete%20Tags :Operator Substring :SelectedTags "Type:Point, Type:Spline, Type:ClosedSpline, Type:Polygon2D, Type:Surface, Type:Texture2D, Type:Volume"))
  (bind AddTags_3 (Add%20Tags :TagsToAdd "Type:ClosedSpline"))
  (bind Gather_2 (Gather))
  (bind DataNum_6 (Data%20Count))
  (bind AttributeBooleanOp_14 (Attribute%20Boolean%20Op :InA InputsAllowClosedSplines))
  (bind Branch_9 (Branch))
  (bind DeleteTags_3 (Delete%20Tags :Operator Substring :SelectedTags "Type:Point, Type:Spline, Type:ClosedSpline, Type:Polygon2D, Type:Surface, Type:Texture2D, Type:Volume"))
  (bind AddTags_2 (Add%20Tags :TagsToAdd "Type:Spline"))
  (bind Gather_1 (Gather))
  (bind DataNum_7 (Data%20Count))
  (bind AttributeBooleanOp_8 (Attribute%20Boolean%20Op :InA InputsAllowSplines))
  (bind Branch_10 (Branch))
  (bind FilterDataByType_9 (Filter%20Data%20By%20Type :TargetType (struct :Ids [(struct :Struct /Script/PCG.PCGDataTypeInfoPolygon2D)])))
  (bind FilterDataByTag_0 (Filter%20Data%20By%20Tag :SelectedTags "Type:Polygon2D"))
  (bind DeleteTags_5 (Delete%20Tags :Operator Substring :SelectedTags "Type:Point, Type:Spline, Type:ClosedSpline, Type:Polygon2D, Type:Surface, Type:Texture2D, Type:Volume"))
  (bind AddTags_4 (Add%20Tags :TagsToAdd "Type:Polygon2D"))
  (bind Gather_3 (Gather))
  (bind DataNum_0 (Data%20Count))
  (bind AttributeBooleanOp_0 (Attribute%20Boolean%20Op :InA InputsAllowPolygon2Ds))
  (bind Branch_0 (Branch))
  (bind FilterDataByType_10 (Filter%20Data%20By%20Type :TargetType (struct :Ids [(struct :Struct /Script/PCG.PCGDataTypeInfoSurface)])))
  (bind FilterDataByTag_1 (Filter%20Data%20By%20Tag :SelectedTags "Type:Surface"))
  (bind DeleteTags_6 (Delete%20Tags :Operator Substring :SelectedTags "Type:Point, Type:Spline, Type:ClosedSpline, Type:Polygon2D, Type:Surface, Type:Texture2D, Type:Volume"))
  (bind AddTags_5 (Add%20Tags :TagsToAdd "Type:Surface"))
  (bind Gather_4 (Gather))
  (bind DataNum_2 (Data%20Count))
  (bind AttributeBooleanOp_1 (Attribute%20Boolean%20Op :InA InputsAllowSurfaces))
  (bind Branch_2 (Branch))
  (bind FilterDataByType_11 (Filter%20Data%20By%20Type :TargetType (struct :Ids [(struct :Struct /Script/PCG.PCGDataTypeInfoTexture2DSingleBase)])))
  (bind FilterDataByTag_5 (Filter%20Data%20By%20Tag :SelectedTags "Type:Texture2D"))
  (bind DeleteTags_7 (Delete%20Tags :Operator Substring :SelectedTags "Type:Point, Type:Spline, Type:ClosedSpline, Type:Polygon2D, Type:Surface, Type:Texture2D, Type:Volume"))
  (bind AddTags_6 (Add%20Tags :TagsToAdd "Type:Texture2D"))
  (bind Gather_5 (Gather))
  (bind DataNum_3 (Data%20Count))
  (bind AttributeBooleanOp_5 (Attribute%20Boolean%20Op :InA InputsAllowTexture2Ds))
  (bind Branch_3 (Branch))
  (bind FilterDataByType_12 (Filter%20Data%20By%20Type :TargetType (struct :Ids [(struct :Struct /Script/PCG.PCGDataTypeInfoVolume)])))
  (bind FilterDataByTag_6 (Filter%20Data%20By%20Tag :SelectedTags "Type:Volume"))
  (bind DeleteTags_8 (Delete%20Tags :Operator Substring :SelectedTags "Type:Point, Type:Spline, Type:ClosedSpline, Type:Polygon2D, Type:Surface, Type:Texture2D, Type:Volume"))
  (bind AddTags_7 (Add%20Tags :TagsToAdd "Type:Volume"))
  (bind Gather_6 (Gather))
  (bind DataNum_5 (Data%20Count))
  (bind AttributeBooleanOp_6 (Attribute%20Boolean%20Op :InA InputsAllowVolumes))
  (bind Branch_4 (Branch))
  (connect DeleteTags_2 Out AddTags_1 In)
  (connect DeleteTags_3 Out AddTags_2 In)
  (connect DeleteTags_4 Out AddTags_3 In)
  (connect DeleteTags_5 Out AddTags_4 In)
  (connect DeleteTags_6 Out AddTags_5 In)
  (connect DeleteTags_7 Out AddTags_6 In)
  (connect DeleteTags_8 Out AddTags_7 In)
  (connect DataNum_0 Out AttributeBooleanOp_0 InB)
  (connect DataNum_2 Out AttributeBooleanOp_1 InB)
  (connect DataNum_3 Out AttributeBooleanOp_5 InB)
  (connect DataNum_5 Out AttributeBooleanOp_6 InB)
  (connect DataNum_7 Out AttributeBooleanOp_8 InB)
  (connect DataNum_4 Out AttributeBooleanOp_12 InB)
  (connect DataNum_6 Out AttributeBooleanOp_14 InB)
  (connect AttributeBooleanOp_0 Out Branch_0 bOutputToB)
  (connect Gather_3 Out Branch_0 In)
  (connect AttributeBooleanOp_1 Out Branch_2 bOutputToB)
  (connect Gather_4 Out Branch_2 In)
  (connect AttributeBooleanOp_5 Out Branch_3 bOutputToB)
  (connect Gather_5 Out Branch_3 In)
  (connect AttributeBooleanOp_6 Out Branch_4 bOutputToB)
  (connect Gather_6 Out Branch_4 In)
  (connect AttributeBooleanOp_12 Out Branch_8 bOutputToB)
  (connect Gather_0 Out Branch_8 In)
  (connect AttributeBooleanOp_14 Out Branch_9 bOutputToB)
  (connect Gather_2 Out Branch_9 In)
  (connect AttributeBooleanOp_8 Out Branch_10 bOutputToB)
  (connect Gather_1 Out Branch_10 In)
  (connect Gather_3 Out DataNum_0 In)
  (connect Gather_4 Out DataNum_2 In)
  (connect Gather_5 Out DataNum_3 In)
  (connect Gather_0 Out DataNum_4 In)
  (connect Gather_6 Out DataNum_5 In)
  (connect Gather_2 Out DataNum_6 In)
  (connect Gather_1 Out DataNum_7 In)
  (connect FilterDataByTag_9 OutsideFilter DeleteTags_2 In)
  (connect FilterDataByTag_4 OutsideFilter DeleteTags_3 In)
  (connect FilterDataByTag_3 OutsideFilter DeleteTags_4 In)
  (connect FilterDataByTag_0 OutsideFilter DeleteTags_5 In)
  (connect FilterDataByTag_1 OutsideFilter DeleteTags_6 In)
  (connect FilterDataByTag_5 OutsideFilter DeleteTags_7 In)
  (connect FilterDataByTag_6 OutsideFilter DeleteTags_8 In)
  (connect FilterDataByType_8 InsideFilter FilterDataByAttribute_1 In)
  (connect FilterDataByType_9 InsideFilter FilterDataByTag_0 In)
  (connect FilterDataByType_10 InsideFilter FilterDataByTag_1 In)
  (connect FilterDataByAttribute_1 OutsideFilter FilterDataByTag_3 In)
  (connect FilterDataByAttribute_1 InsideFilter FilterDataByTag_4 In)
  (connect FilterDataByType_11 InsideFilter FilterDataByTag_5 In)
  (connect FilterDataByType_12 InsideFilter FilterDataByTag_6 In)
  (connect FilterDataByType_7 InsideFilter FilterDataByTag_9 In)
  (connect DefaultInputNode InputReferences FilterDataByType_7 In)
  (connect DefaultInputNode InputReferences FilterDataByType_8 In)
  (connect DefaultInputNode InputReferences FilterDataByType_9 In)
  (connect DefaultInputNode InputReferences FilterDataByType_10 In)
  (connect DefaultInputNode InputReferences FilterDataByType_11 In)
  (connect DefaultInputNode InputReferences FilterDataByType_12 In)
  (connect FilterDataByTag_9 InsideFilter Gather_0 In)
  (connect AddTags_1 Out Gather_0 In2)
  (connect FilterDataByTag_4 InsideFilter Gather_1 In)
  (connect AddTags_2 Out Gather_1 In2)
  (connect FilterDataByTag_3 InsideFilter Gather_2 In)
  (connect AddTags_3 Out Gather_2 In2)
  (connect FilterDataByTag_0 InsideFilter Gather_3 In)
  (connect AddTags_4 Out Gather_3 In2)
  (connect FilterDataByTag_1 InsideFilter Gather_4 In)
  (connect AddTags_5 Out Gather_4 In2)
  (connect FilterDataByTag_5 InsideFilter Gather_5 In)
  (connect AddTags_6 Out Gather_5 In2)
  (connect FilterDataByTag_6 InsideFilter Gather_6 In)
  (connect AddTags_7 Out Gather_6 In2)
  (output ClosedSplines Branch_9.Output%20B)
  (output Points Branch_8.Output%20B)
  (output Polygon2Ds Branch_0.Output%20B)
  (output Splines Branch_10.Output%20B)
  (output Surfaces Branch_2.Output%20B)
  (output Texture2Ds Branch_3.Output%20B)
  (output Volumes Branch_4.Output%20B))
