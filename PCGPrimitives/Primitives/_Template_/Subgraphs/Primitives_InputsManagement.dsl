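; PCG subgraph in the primitives template that manages a primitive's inputs, turning input references into processed data or fallback data.
; It strips empty data, then a count-driven branch sends references into type filtering, gated by per-type allow flags for points, splines, surfaces, volumes and others.
; Typed results feed an overridable processing graph.
; An overridable fallback graph runs through execution dependencies when references or processed results are empty, and both graphs feed the output.
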(graph Primitives_InputsManagement
  (params (InputsProcessingGraph :type (SoftObject /Script/PCG.PCGGraphInterface) :default /PCGPrimitives/Primitives/_Template_/Primitives_InputsProcessing.Primitives_InputsProcessing) (InputsFallbackGraph :type (SoftObject /Script/PCG.PCGGraphInterface) :default /PCGPrimitives/Primitives/_Template_/Primitives_InputsFallback.Primitives_InputsFallback) (InputsAllowPoints :type Bool :default true) (InputsAllowSplines :type Bool :default true) (InputsAllowClosedSplines :type Bool :default true) (InputsAllowPolygon2Ds :type Bool :default true) (InputsAllowSurfaces :type Bool :default true) (InputsAllowTexture2Ds :type Bool :default true) (InputsAllowVolumes :type Bool :default true))
  (bind DataNum_2 (Data%20Count))
  (bind Branch_1 (Branch))
  ;[0 0 0 0.5] 1 - Inputs
  (bind RemoveEmptyData_4 (Remove%20Empty%20Data))
  ;[0 0 0 0.5] 2 - Use Primary and Secondary Input References lists and filter the desired working data types
  (bind DataNum_1 (Data%20Count))
  (bind Branch_2 (Branch))
  (bind Subgraph_2 (Subgraph :Graph /PCGPrimitives/Primitives/_Template_/Subgraphs/Primitives_InputsDataFiltering.Primitives_InputsDataFiltering :InputsAllowClosedSplines InputsAllowClosedSplines :InputsAllowPoints InputsAllowPoints :InputsAllowPolygon2Ds InputsAllowPolygon2Ds :InputsAllowSplines InputsAllowSplines :InputsAllowSurfaces InputsAllowSurfaces :InputsAllowTexture2Ds InputsAllowTexture2Ds :InputsAllowVolumes InputsAllowVolumes))
  ;[0 0 0 0.5] 3 - Use or Transform inputs
  (bind Subgraph_0 (Subgraph :Graph /PCGPrimitives/Primitives/_Template_/Primitives_InputsProcessing.Primitives_InputsProcessing :SubgraphOverride-param InputsProcessingGraph))
  ;[0 0 0 0.5] 4 - Fallback
  (bind Subgraph_1 (Subgraph :Graph /PCGPrimitives/Primitives/_Template_/Primitives_InputsFallback.Primitives_InputsFallback :SubgraphOverride-param InputsFallbackGraph))
  (reroute EmptyInputReferences Branch_2.Output%20A)
  (connect DataNum_2 Out Branch_1 bOutputToB)
  (connect DataNum_2 Out Branch_1 In)
  (connect DataNum_1 Out Branch_2 bOutputToB)
  (connect RemoveEmptyData_4 Out Branch_2 In)
  (connect RemoveEmptyData_4 Out DataNum_1 In)
  (connect Subgraph_0 Out DataNum_2 In)
  (connect DefaultInputNode In RemoveEmptyData_4 In)
  (connect Subgraph_2 ClosedSplines Subgraph_0 ClosedSplines)
  (connect Subgraph_2 Points Subgraph_0 Points)
  (connect Subgraph_2 Polygon2Ds Subgraph_0 Polygon2Ds)
  (connect Subgraph_2 Splines Subgraph_0 Splines)
  (connect Subgraph_2 Surfaces Subgraph_0 Surfaces)
  (connect Subgraph_2 Texture2Ds Subgraph_0 Texture2Ds)
  (connect Subgraph_2 Volumes Subgraph_0 Volumes)
  (connect Branch_1 Output%20A Subgraph_1 Execution%20Dependency)
  (connect EmptyInputReferences Out Subgraph_1 Execution%20Dependency)
  (connect Branch_2 Output%20B Subgraph_2 InputReferences)
  (output Out Subgraph_0)
  (output Out Subgraph_1))
